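; Looping gameplay cue for the equipped chainsaw that plays its turn-on VFX when applied.
; On application it calls the parent, then checks that the target player pawn's weapon is not hidden by holstering.
; If visible, it casts the current weapon to the chainsaw Blueprint and spawns an auto-released Niagara system attached to the weapon mesh's root.
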(event EventOnApplication(EitherVFX-preferthisone)
  (params (MyTarget :type Actor%20Object%20Reference) (Parameters :type Gameplay%20Cue%20Parameters%20Structure%20%28by%20ref%29) (ParticleComponents :type Array%20of%20FXSystem%20Component%20Object%20References) (AudioComponents :type Array%20of%20Audio%20Component%20Object%20References) (BurstCameraShakeInstance :type Camera%20Shake%20Base%20Object%20Reference) (BurstDecalInstance :type Decal%20Actor%20Object%20Reference))
  (|CallParentFunction MyTarget Parameters ParticleComponents AudioComponents BurstCameraShakeInstance BurstDecalInstance)
  (bind _asfort_player_pawn (Utilities|Casting|CastToFortPlayerPawn MyTarget))
  (if (not (Weapon|IsWeaponHiddenbyHolstering _asfort_player_pawn 0))
    (bind _asb_weapon_chainsaw (Utilities|Casting|CastToB_Weapon_Chainsaw (Weapon|GetCurrentWeapon _asfort_player_pawn)))
    (bind _weaponmesh (Variables|/Script/FortniteGame.FortWeapon|GetWeaponMesh _asb_weapon_chainsaw)
      (:then
        (Niagara|SpawnSystemAttached "/ChainsawGameplay/Assets/FX/NS_Chainsaw_TurnOn.NS_Chainsaw_TurnOn" _weaponmesh "root" "0, 0, 0" "0, 0, 0" "SnapToTarget" true true "AutoRelease"))
      (:else))))

(fn ConstructionScript)
