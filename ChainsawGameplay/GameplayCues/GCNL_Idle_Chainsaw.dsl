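; Looping gameplay cue for the chainsaw that turns the weapon's idle Niagara effect on and off.
; On application it finds the target pawn's current chainsaw and activates its IdleFX component.
; If the weapon cast fails, a stored next-tick timer retries until the cast succeeds, reusing the target from the application event.
; On removal it clears that timer and deactivates the effect, assuming the weapon is still relevant.
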(event EventOnApplication(EitherVFX-preferthisone)
  (params (MyTarget :type Actor%20Object%20Reference) (Parameters :type Gameplay%20Cue%20Parameters%20Structure%20%28by%20ref%29) (ParticleComponents :type Array%20of%20FXSystem%20Component%20Object%20References) (AudioComponents :type Array%20of%20Audio%20Component%20Object%20References) (BurstCameraShakeInstance :type Camera%20Shake%20Base%20Object%20Reference) (BurstDecalInstance :type Decal%20Actor%20Object%20Reference))
  ;[0.04057 0.15 0.05877 0.5] OnApplication: Activate the idle niagara fx component on the weapon.
  (|CallParentFunction MyTarget Parameters ParticleComponents AudioComponents BurstCameraShakeInstance BurstDecalInstance)
  (Utilities|IsValid MyTarget
    (:Is%20Valid
      (Utilities|Time|ClearandInvalidateTimerbyHandle MyTarget (Variables|Default|GetWaitingForWeaponTimerHandle))
      (bind _asfort_player_pawn (Utilities|Casting|CastToFortPlayerPawn MyTarget))
      (bind _asb_weapon_chainsaw (Utilities|Casting|CastToB_Weapon_Chainsaw (Weapon|GetCurrentWeapon _asfort_player_pawn))
        (:then
          (bind _idlefx_niagara (Variables|/Game/Weapons/FORT_Melee/Blueprints/B_Melee_Generic.B_Melee_Generic_C|GetIdleFX%20Niagara _asb_weapon_chainsaw)
            (:then
              (Components|Activation|Activate _idlefx_niagara false))
            (:else)))
        (:CastFailed
          (bind _returnvalue (Utilities|Time|SetTimerforNextTickbyEvent (delegate Custom|WaitingForWeaponActivate)))
          (Variables|Default|SetWaitingForWeaponTimerHandle _returnvalue))))))

(event Custom|WaitingForWeaponActivate
  ;[0.04057 0.15 0.05877 0.5] OnApplication: Activate the idle niagara fx component on the weapon.
  (bind _asfort_player_pawn (Utilities|Casting|CastToFortPlayerPawn (stale EventOnApplication(EitherVFX-preferthisone) MyTarget)))
  (bind _asb_weapon_chainsaw (Utilities|Casting|CastToB_Weapon_Chainsaw (Weapon|GetCurrentWeapon _asfort_player_pawn))
    (:then
      (bind _idlefx_niagara (Variables|/Game/Weapons/FORT_Melee/Blueprints/B_Melee_Generic.B_Melee_Generic_C|GetIdleFX%20Niagara _asb_weapon_chainsaw)
        (:then
          (Components|Activation|Activate _idlefx_niagara false))
        (:else)))
    (:CastFailed
      (bind _returnvalue (Utilities|Time|SetTimerforNextTickbyEvent (delegate Custom|WaitingForWeaponActivate)))
      (Variables|Default|SetWaitingForWeaponTimerHandle _returnvalue))))

(event EventOnRemoval(EitherVFX-preferthisversion)
  (params (MyTarget :type Actor%20Object%20Reference) (Parameters :type Gameplay%20Cue%20Parameters%20Structure%20%28by%20ref%29) (ParticleComponents :type Array%20of%20FXSystem%20Component%20Object%20References) (AudioComponents :type Array%20of%20Audio%20Component%20Object%20References) (BurstCameraShakeInstance :type Camera%20Shake%20Base%20Object%20Reference) (BurstDecalInstance :type Decal%20Actor%20Object%20Reference))
  ;[0.1128 0.06717 0.01553 0.5] OnRemoval: Deactivate the idle niagara fx component.
  (|CallParentFunction MyTarget Parameters ParticleComponents AudioComponents BurstCameraShakeInstance BurstDecalInstance)
  (Utilities|IsValid MyTarget
    (:Is%20Valid
      (Utilities|Time|ClearandInvalidateTimerbyHandle MyTarget (Variables|Default|GetWaitingForWeaponTimerHandle))
      (bind _asfort_player_pawn (Utilities|Casting|CastToFortPlayerPawn MyTarget))
      (bind _asb_weapon_chainsaw (Utilities|Casting|CastToB_Weapon_Chainsaw (Weapon|GetCurrentWeapon _asfort_player_pawn))) ; Weapon should be relevant on removal.
      (bind _idlefx_niagara (Variables|/Game/Weapons/FORT_Melee/Blueprints/B_Melee_Generic.B_Melee_Generic_C|GetIdleFX%20Niagara _asb_weapon_chainsaw)
        (:then
          (Components|Activation|Deactivate _idlefx_niagara))
        (:else)))))

(fn ConstructionScript)
