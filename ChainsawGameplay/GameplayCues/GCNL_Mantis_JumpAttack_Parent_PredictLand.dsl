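; Parent gameplay cue for the Mantis jump attack that spawns a particle effect when the target pawn lands.
; On execute it binds the pawn's landed and movement-mode events, and the effect appears once, slightly ahead of the pawn at foot height.
; An async wait for the clambering ledge-moving tag suppresses the effect.
; Leaving surface swimming or skydiving cancels the landing listener, as does the cue ceasing relevance.
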(event EventBeginPlay
  (Variables|Default|SetSpawnedParticleSystem 0)
  (Variables|Default|SetFortPawn 0))

(event Custom|UndbindMovementModeChanged
  (Character|UnbindEventfromMovementModeChangedDelegate (Variables|Default|GetFortPawn) (delegate Custom|MovementModeChangedDelegate_Event_0)))

(event Custom|MovementModeChangedDelegate_Event_0
  (params (Character :type Character%20Object%20Reference) (PrevMovementMode :type EMovementMode%20Enum :default "MOVE_None") (PreviousCustomMode :type Byte))
  (switch Utilities|FlowControl|Switch|SwitchonEMovementMode PrevMovementMode
    (:MOVE_Custom
      (if (or (Utilities|Enum|Equal(Enum) (Utilities|Enum|BytetoEnumEFortCustomMovement PreviousCustomMode) (Variables|Default|GetSurfaceSwimmingMovementMode)) (Utilities|Enum|Equal(Enum) (Utilities|Enum|BytetoEnumEFortCustomMovement PreviousCustomMode) (Variables|Default|GetSkydivingMovementMode)))
        (CallFunction|UndbindMovementModeChanged)
        (CallFunction|UnbindLanded)))))

(event Custom|ExecutionStart
  (params (FortPawn :type Fort%20Pawn%20Object%20Reference))
  (bind _output_get (Variables|Default|SetFortPawn FortPawn))
  (CallFunction|Clambering _output_get)
  (Variables|Default|SetClambering%20Start false)
  (EventDispatchers|AssignOnPawnLanded _output_get (delegate Custom|OnPawnLanded))
  (Character|AssignMovementModeChangedDelegate (Variables|Default|GetFortPawn) (delegate Custom|MovementModeChangedDelegate_Event_0)))

(event Custom|OnPawnLanded
  (params (Hit :type Hit%20Result%20Structure%20%28by%20ref%29))
  (bind (_returnvalue_x _returnvalue_y _returnvalue_z) (Math|Vector|BreakVector (Transformation|GetWorldLocation (Variables|/Script/Engine.Character|GetCapsuleComponent (Variables|Default|GetFortPawn)))))
  (bind (_returnvalue_x_2 _returnvalue_y_2) (Math|Vector|BreakVector (+ (Transformation|GetActorLocation (Variables|Default|GetFortPawn)) (* (Math|Rotator|GetRotationXVector (Transformation|GetActorRotation (Variables|Default|GetFortPawn))) 1.5e+02))))
  (CallFunction|UnbindLanded)
  (if (not (Variables|Default|GetClambering%20Start))
    (bind _returnvalue (Effects|Components|ParticleSystem|SpawnFXAtLocation (Variables|Default|GetFortPawn) (Variables|Default|GetParticleSystem) (Math|Vector|MakeVector _returnvalue_x_2 _returnvalue_y_2 (- _returnvalue_z (Components|Capsule|GetUnscaledCapsuleHalfHeight (Variables|/Script/Engine.Character|GetCapsuleComponent (Variables|Default|GetFortPawn))))) (Transformation|GetActorRotation (Variables|Default|GetFortPawn)) "1.000000,1.000000,1.000000" true "None" true))
    (Variables|Default|SetSpawnedParticleSystem _returnvalue)
    (else
      _))
  (Ability|Async|EndAction (Variables|Default|GetClamberingAction)))

(event Custom|UnbindLanded
  (EventDispatchers|UnbindEventfromOnPawnLanded (Variables|Default|GetFortPawn) (delegate Custom|OnPawnLanded)))

(event Custom|Clambering
  (params (TargetActor :type Actor%20Object%20Reference))
  (bind _asynctaskproxy (Ability|Tasks|WaitGameplayTagAddToActor TargetActor "(TagName=\"Clambering.State.LedgeMoving\")")
    (:then
      (Variables|Default|SetClamberingAction _asynctaskproxy))
    (:Added
      (Variables|Default|SetClambering%20Start true))))

(fn OnBurst
  (params (MyTarget :type Actor%20Object%20Reference) (Parameters :type Gameplay%20Cue%20Parameters%20Structure%20%28by%20ref%29))
  (return false))

(fn OnExecute
  (params (MyTarget :type Actor%20Object%20Reference) (Parameters :type Gameplay%20Cue%20Parameters%20Structure%20%28by%20ref%29))
  (bind _asfort_pawn (Utilities|Casting|CastToFortPawn MyTarget)
    (:then
      (CallFunction|ExecutionStart _asfort_pawn)
      (return true))
    (:CastFailed
      (return false))))

(fn OnCeaseRelevant
  (params (MyTarget :type Actor%20Object%20Reference) (Parameters :type Gameplay%20Cue%20Parameters%20Structure%20%28by%20ref%29))
  (CallFunction|UnbindLanded)
  (return true))

(fn ConstructionScript)
