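; Looping gameplay cue for chainsaw sprinting that silences and hides its effects while the player's weapon is holstered.
; On loop start it caches the spawned particle and audio components, binds to the pawn's holster and unholster events, and applies the holstered state if already holstered.
; The handlers toggle particle visibility and audio volume instead of deactivating, because reactivation misbehaves on unholster.
; Removal unbinds both events.
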(event EventOnLoopingStart(EitherVFX-preferthisversion)
  (params (MyTarget :type Actor%20Object%20Reference) (Parameters :type Gameplay%20Cue%20Parameters%20Structure%20%28by%20ref%29) (ParticleComponents :type Array%20of%20FXSystem%20Component%20Object%20References) (AudioComponents :type Array%20of%20Audio%20Component%20Object%20References))
  ;[0.01686 0.15 0.03001 0.5] OnLoopingStart: Cache the fx components and bind to know whenever the weapon is holstered and unholstered.
  (|CallParentFunction MyTarget Parameters ParticleComponents AudioComponents)
  (Variables|Transient|SetParticleComponents ParticleComponents)
  (Variables|Transient|SetAudioComponents AudioComponents)
  (bind _asfort_player_pawn (Utilities|Casting|CastToFortPlayerPawn MyTarget))
  (EventDispatchers|AssignOnHolsteredEvent _asfort_player_pawn (EventDispatchers|CreateEvent :!event HandleWeaponHolstered))
  (EventDispatchers|BindEventtoOnUnholsteredEvent _asfort_player_pawn (EventDispatchers|CreateEvent :!event HandleWeaponUnholstered))
  (if (Weapon|IsWeaponHolstered _asfort_player_pawn)
    (CallFunction|HandleWeaponHolstered)))

(event EventOnRemoval(EitherVFX-preferthisversion)
  (params (MyTarget :type Actor%20Object%20Reference) (Parameters :type Gameplay%20Cue%20Parameters%20Structure%20%28by%20ref%29) (ParticleComponents :type Array%20of%20FXSystem%20Component%20Object%20References) (AudioComponents :type Array%20of%20Audio%20Component%20Object%20References) (BurstCameraShakeInstance :type Camera%20Shake%20Base%20Object%20Reference) (BurstDecalInstance :type Decal%20Actor%20Object%20Reference))
  ;[0.15 0.07882 0.01507 0.5] OnRemoval: Unbind delegates.
  (|CallParentFunction MyTarget Parameters ParticleComponents AudioComponents BurstCameraShakeInstance BurstDecalInstance)
  (bind _asfort_player_pawn (Utilities|Casting|CastToFortPlayerPawn MyTarget))
  (EventDispatchers|UnbindEventfromOnHolsteredEvent _asfort_player_pawn (EventDispatchers|CreateEvent :!event HandleWeaponHolstered))
  (EventDispatchers|UnbindEventfromOnUnholsteredEvent _asfort_player_pawn (EventDispatchers|CreateEvent :!event HandleWeaponUnholstered)))

(fn HandleWeaponHolstered
  ;[0.03631 0.15 0.1285 0.5] HandleWeaponHolstered: Weapon has been holstered, disable the spawned FX components.
  ; Not using Deactivate here since Activate() doesn't work as expected when the weapon is unholstered.
  (Rendering|SetVisibility (Variables|Transient|GetParticleComponents) false false)
  (Audio|Components|Audio|SetVolumeMultiplier (Variables|Transient|GetAudioComponents) 0.0 :!class /Script/Engine.AudioComponent))

(fn HandleWeaponUnholstered
  ; HandleWeaponUnholstered: Turn the disabled fx components back on.
  (Rendering|SetVisibility (Variables|Transient|GetParticleComponents) true false)
  (Audio|Components|Audio|SetVolumeMultiplier (Variables|Transient|GetAudioComponents) 1.0 :!class /Script/Engine.AudioComponent))

(fn ConstructionScript)
